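# MetaSound Source preset of MSS_MusicEncounter_NPC_Default that supplies the NarrowFlea NPC encounter music, spatialized with NPC music attenuation.
# It fills the parent's Idle, Combat and Defeated slots with a phase 0 loop, a phase 1 loop plus entry stinger, and a death stinger.
# Exposed phase, tempo, beat offset, fade and pause inputs parameterize the transitions.
# The Combat loop occupies only the percussion slot, leaving the second layer and the Idle transition empty.
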MetaSoundSource /SouthBulkAI/Audio/Music/MSS_IP_Combat_NarrowFlea
preset_parent=/MusicEncounter/MetaSounds/MSS_MusicEncounter_NPC_Default
inputs [
 Combat.BeatOffset {type=Int32}
 Combat.ElseWave {type=WaveAsset}
 Combat.FadeInTime {type=Time}
 Combat.FadeOutTime {type=Time}
 Combat.PercWave {type=WaveAsset}
 Combat.TransitionVolume {type=Float}
 Combat.TransitionWave {type=WaveAsset}
 Combat.Volume {type=Float}
 Defeated.BeatOffset {type=Int32}
 Defeated.Volume {type=Float}
 Defeated.Wave {type=WaveAsset}
 Idle.BeatOffset {type=Int32}
 Idle.FadeInTime {type=Time}
 Idle.FadeOutTime {type=Time}
 Idle.TransitionVolume {type=Float}
 Idle.TransitionWave {type=WaveAsset}
 Idle.Volume {type=Float}
 Idle.Wave {type=WaveAsset}
 MusicEncounter.bIsPaused {type=Bool}
 MusicEncounter.OutputVolume {type=Float}
 MusicEncounter.PauseFadeOutTime {type=Time}
 MusicEncounter.Phase {type=Int32}
 MusicEncounter.Tempo {type=Float}
 UE.Source.OnPlay {type=Trigger}
]
interfaces [UE.OutputFormat.Stereo:1.0, UE.Source.OneShot:1.0, UE.Source:1.1]
outputs [
 "UE.OutputFormat.Stereo.Audio:0" {type=Audio}
 "UE.OutputFormat.Stereo.Audio:1" {type=Audio}
 UE.Source.OneShot.OnFinished {type=Trigger}
]
pages [Default {input_defaults {
   Combat.BeatOffset=5
   Combat.ElseWave=null
   Combat.FadeInTime=0
   Combat.FadeOutTime=1
   Combat.PercWave=&/MX_IP_S42_Combat_NarrowFlea_Phase_1_LP
   Combat.TransitionVolume=0
   Combat.TransitionWave=&/MX_IP_S42_Combat_NarrowFlea_Phase_0_to_1_STNG
   Combat.Volume=0
   Defeated.BeatOffset=0
   Defeated.Volume=0
   Defeated.Wave=&/MX_IP_S42_Combat_NarrowFlea_Death_STNG
   Idle.BeatOffset=1
   Idle.FadeInTime=1
   Idle.FadeOutTime=1
   Idle.TransitionVolume=0
   Idle.TransitionWave=null
   Idle.Volume=0
   Idle.Wave=&/MX_IP_S42_Combat_NarrowFlea_Phase_0_LP
   MusicEncounter.OutputVolume=0
   MusicEncounter.PauseFadeOutTime=3
   MusicEncounter.Phase=0
   MusicEncounter.Tempo=133
   MusicEncounter.bIsPaused=null
   UE.Source.OnPlay=null
  }}]
properties {
 assetImportData=&self:AssetImportData
 attenuationSettings=/Game/Sounds/Attenuation/ATT_Music_NPC
 maxDistance=6500
 numChannels=2
 outputFormat=Stereo
 sampleRate=48000
 soundClassObject=&Fort_Music_NPC
 assetUserData [&self:AudioAssetUserData_0]
}
